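# Niagara system for the Supernova vehicle boost cosmetic, emitting a trail behind the car with painted-variant color overrides.
# Two ribbon emitters lay tapering trails, the second adding a separately sized black-ribbon width and multiple renderers.
# A sprite-and-mesh emitter and a two-color sprite emitter use spawn groups as visibility tags to drive several layers.
# Ribbon velocities switch per game mode through mask modules, and spawn offsets follow User.DTCarOffset.
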NiagaraSystem /VehicleCosmetics/Boosts/Boost_Supernova/FX/NS_Supernova
dependencies {
 usedDataInterfaces [
  &NiagaraDataInterfaceCurve
  &NiagaraDataInterfaceVectorCurve
  &NiagaraDataInterfaceArrayColor
  &NiagaraDataInterfaceSkeletalMesh
  &NiagaraDataInterfaceArrayFloat3
 ]
 usedDynamicInputs [&SelectVectorFromArray]
 usedModules [
  &self:SystemSpawnScript.SystemScriptSource.SystemScriptGraph.NiagaraNodeAssignment_0.SetVariables_#1
  &self:SystemSpawnScript.SystemScriptSource.SystemScriptGraph.NiagaraNodeAssignment_1.SetVariables_#2
  &self:SystemSpawnScript.SystemScriptSource.SystemScriptGraph.NiagaraNodeAssignment_2.SetVariables_#3
  &SystemState
 ]
}
emitters {
 NE_VC_SingleBoost_Ribbon_Parent {
  emitterUpdateScript [
   &EmitterState()
   &VC_FilteredSocketCount()
   &VC_RateQualityHandler(Default Rate:NiagaraFloat=&Multiply_Float(
     A=60
     B=/VehicleCosmetics/FX/Niagara/Modules/VC_Mask_Floats(Coro Value:NiagaraFloat=1, FrontEnd Value:NiagaraFloat=0.75, FN Value:NiagaraFloat=0.5)
    ))
   &VC_SpawnRibbon_v1(SpawnRate=$Emitter.Spawn Rate)
  ]
  particleSpawnScript [
   &VC_PaintedVariantsOverride(DefaultColor=<rgba 1.2167 0.217 2.6042 1>)
   &VC_InitializeRibbon_v1(
    Color=$Particles.Color
    Custom Offset=true
    Lifetime=0.6
    Local Vector Offset=&AddVectorToPosition(Position=<-250 0 20>, Vector=$User.DTCarOffset)
    Offset Amount=1
    Ribbon Width=50
   )
   &AddVelocity(Velocity=&VC_Mask_Vectors(Coro Value:Vector3f=<-400 0 0>, FrontEnd Value:Vector3f=<-3000 0 0>, FN Value:Vector3f=<-1200 0 0>)) {module=AddVelocity001}
   /VehicleCosmetics/FX/VC_FXParents/Modules/DLR_ProductPreviews(GameModeID:NiagaraInt32=$System.GameModeID, bPreviewVehicleSpeed:NiagaraBool=$System.bPreviewVehicleSpeed)
  ]
  particleUpdateScript [
   &ParticleState()
   &InheritSourceMovement(
    Applied Position Scale=&VectorFromFloat(Value=&FloatFromCurve(FloatCurve=DataInterface() {propertyValues {Curve {
         defaultValue=1
         keys [
          {arriveTangent=-1.7885, interpMode=RCIM_Cubic, leaveTangent=-1.7885, tangentMode=RCTM_User, time=0.05, value=1}
          {interpMode=RCIM_Cubic, tangentMode=RCTM_User, time=1, value=0}
         ]
        }}}))
    ApplyForce=false
    ApplyPosition=true
    ApplyPositionFalloff=false
   )
   &SolveForcesAndVelocity(Clamp Velocity=false, Limit Acceleration=false)
   &VC_UpdateRibbon_v1()
   &ScaleRibbonWidth(Ribbon Width Scale=&FloatFromCurve(FloatCurve=DataInterface() {propertyValues {Curve {
        defaultValue=1
        keys [
         {interpMode=RCIM_Cubic, tangentMode=RCTM_User, time=0, value=0}
         {interpMode=RCIM_Cubic, tangentMode=RCTM_User, time=0.5, value=0.25}
         {tangentMode=RCTM_User, time=1, value=0}
        ]
       }}}))
  ]
  rendererClasses [&NiagaraRibbonRendererProperties]
  renderers [{rendererClass=&NiagaraRibbonRendererProperties}]
 }
 NE_VC_SingleBoost_Ribbon_Parent001 {
  emitterUpdateScript [
   &EmitterState()
   &VC_FilteredSocketCount()
   &VC_RateQualityHandler(Epic Spawn Rate=8, High Spawn Rate=8, Low Spawn Rate=8, Medium Spawn Rate=8, Use Custom Rate?=true)
   &VC_SpawnRibbon_v1(SpawnRate=$Emitter.Spawn Rate)
  ]
  particleSpawnScript [
   &VC_PaintedVariantsOverride(
    DefaultColor=<rgba 0.1302 4.3032 5 1>
    R1 DefaultColor=<rgba 0 0 0 1>
    R2 DefaultColor:LinearColor=<rgba 2.3361 0.4167 5 1>
    Use R1 Painted Overrides=true
    Use R2 Painted Overrides=true
   )
   &VC_InitializeRibbon_v1(
    Color=$Particles.Color
    Custom Offset=true
    Lifetime=&RandomRangeFloat(Minimum=0.75, Maximum=1)
    Local Vector Offset=&AddVectorToPosition(Position=<-250 0 25>, Vector=$User.DTCarOffset)
    Offset Amount=1
    Ribbon Width=42.5
   )
   &AddVelocity(Velocity=&VC_Mask_Vectors(Coro Value:Vector3f=<-400 0 0>, FrontEnd Value:Vector3f=<-1500 0 0>, FN Value:Vector3f=<-1200 0 0>)) {module=AddVelocity001}
   &DLR_ProductPreviews(GameModeID:NiagaraInt32=$System.GameModeID, bPreviewVehicleSpeed:NiagaraBool=$System.bPreviewVehicleSpeed)
  ]
  particleUpdateScript [
   &ParticleState()
   &InheritSourceMovement(
    Applied Position Scale=&VectorFromFloat(Value=&FloatFromCurve(FloatCurve=DataInterface() {propertyValues {Curve {
         defaultValue=1
         keys [
          {arriveTangent=-1.7885, interpMode=RCIM_Cubic, leaveTangent=-1.7885, tangentMode=RCTM_User, time=0.05, value=1}
          {interpMode=RCIM_Cubic, tangentMode=RCTM_User, time=1, value=0}
         ]
        }}}))
    ApplyForce=false
    ApplyPosition=true
    ApplyPositionFalloff=false
   )
   &SolveForcesAndVelocity(Clamp Velocity=false, Limit Acceleration=false)
   &VC_UpdateRibbon_v1()
   &ScaleRibbonWidth(Ribbon Width Scale=&FloatFromCurve(FloatCurve=DataInterface() {propertyValues {Curve {
        defaultValue=1
        keys [
         {arriveTangent=0.9193, interpMode=RCIM_Cubic, leaveTangent=0.9193, tangentMode=RCTM_User, time=0, value=0.1}
         {arriveTangent=-0.2568, interpMode=RCIM_Cubic, leaveTangent=-0.2568, tangentMode=RCTM_User, time=0.5, value=0.15}
         {tangentMode=RCTM_User, time=1, value=0.025}
        ]
       }}}))
   /SetVariables(Particles.BlackRibbonWidth:NiagaraFloat=&Multiply_Float(
     A=75
     B=&FloatFromCurve(FloatCurve=DataInterface() {propertyValues {Curve {
         defaultValue=1
         keys [
          {arriveTangent=0.9193, interpMode=RCIM_Cubic, leaveTangent=0.9193, tangentMode=RCTM_User, time=0, value=0.1}
          {arriveTangent=-0.2625, interpMode=RCIM_Cubic, leaveTangent=-0.2625, tangentMode=RCTM_User, time=0.5, value=0.2}
          {tangentMode=RCTM_User, time=1, value=0.05}
         ]
        }}})
    )) {isSetParameters=true}
  ]
  rendererClasses [&NiagaraRibbonRendererProperties]
  renderers [
   {rendererClass=&NiagaraRibbonRendererProperties}
   {rendererClass=&NiagaraRibbonRendererProperties, rendererIndex=1}
   {rendererClass=&NiagaraRibbonRendererProperties, rendererIndex=2}
   {rendererClass=&NiagaraRibbonRendererProperties, rendererIndex=3}
  ]
 }
 NE_VC_SingleBoost_Sprite_Parent {
  emitterUpdateScript [
   &EmitterState()
   &SpawnRate(SpawnRate=6, Use Spawn Probability=false)
   &SpawnRate(SpawnGroup=1, SpawnRate=36, Use Spawn Probability=false) {module=SpawnRate001}
   &SpawnRate(SpawnGroup=2, SpawnRate=3, Use Spawn Probability=false) {module=SpawnRate002}
   &SpawnRate(SpawnGroup=3, SpawnRate=50, Use Spawn Probability=false) {module=SpawnRate003}
  ]
  particleSpawnScript [
   /SetVariables(Particles.VisibilityTag=$Emitter.SpawnGroup) {isSetParameters=true}
   &VC_PaintedVariantsOverride(
    DefaultColor=<rgba 0.952 0 1 0.1652>
    R1 DefaultColor=<rgba 0.013 0.1 0.0906 1>
    R2 DefaultColor:LinearColor=<rgba 0 0.635 1 0.487>
    Use R1 Painted Overrides=true
    Use R2 Painted Overrides=true
   )
   &InitializeParticle(
    Color=$Particles.Color
    Lifetime=&MergeHelper_FloatSwitch(
     0=&RandomRangeFloat(Minimum=0.2, Maximum=0.35)
     1=&RandomRangeFloat(Minimum=0.1, Maximum=&RandomRangeFloat(Minimum=0.1, Maximum=0.2))
     2=&RandomRangeFloat(Minimum=1, Maximum=1)
     3=&RandomRangeFloat(Minimum=0.35, Maximum=0.75)
     4=0
     Index=$Emitter.SpawnGroup
     Num=::NewEnumerator3
    )
    Mesh Scale Mode=::NewEnumerator3
    Mesh Uniform Scale=&MergeHelper_FloatSwitch(
     0=0.5
     1=0.35
     2=&RandomRangeFloat(Minimum=0.125, Maximum=0.2)
     3=0.5
     4=0
     Index=$Emitter.SpawnGroup
     Num=::NewEnumerator3
    )
    Position Offset=&AddVectorToPosition(Position=<-250 0 25>, Vector=$User.DTCarOffset)
    Position Offset Coordinate Space=::Local
    Sprite Rotation Mode=::NewEnumerator2
    UsePositionOffset=true
    Write Material Random=true
   )
   &ShapeLocation(MaskSpawnGroup=false, Override Seed=false, Sphere Radius=1, UseOwnerScale=true)
   &AddVelocity(Velocity=&MergeHelper_VectorSwitch(0=<0 0 0>, 1=<0 0 0>, 2=<0 0 0>, 3=<-1500 0 0>, 4=<0 0 0>, Index=$Emitter.SpawnGroup, Num=::NewEnumerator3))
   &InitialMeshOrientation(Use Rotation Vector=false)
  ]
  particleUpdateScript [
   &ParticleState()
   &ScaleMeshSize(Scale Factor=&MergeHelper_VectorSwitch(
     0=<1 1 1>
     1=<1 1 1>
     2=<1 1 1>
     3=&RandomRangeVector(
      Minimum=&VectorFromCurve(VectorCurve=DataInterface() {propertyValues {
         XCurve {
          defaultValue=1
          keys [
           {arriveTangent=0.65, interpMode=RCIM_Cubic, leaveTangent=0.65, tangentMode=RCTM_User, time=0, value=0.25}
           {arriveTangent=2.1133, interpMode=RCIM_Cubic, leaveTangent=2.1133, tangentMode=RCTM_User, time=1, value=1.25}
          ]
         }
         YCurve {
          defaultValue=1
          keys [
           {arriveTangent=0.65, interpMode=RCIM_Cubic, leaveTangent=0.65, tangentMode=RCTM_User, time=0, value=0.25}
           {arriveTangent=2.1133, interpMode=RCIM_Cubic, leaveTangent=2.1133, tangentMode=RCTM_User, time=1, value=1.25}
          ]
         }
         ZCurve {
          defaultValue=1
          keys [
           {arriveTangent=0.65, interpMode=RCIM_Cubic, leaveTangent=0.65, tangentMode=RCTM_User, time=0, value=0.25}
           {arriveTangent=2.1133, interpMode=RCIM_Cubic, leaveTangent=2.1133, tangentMode=RCTM_User, time=1, value=1.25}
          ]
         }
        }})
      Maximum=&VectorFromCurve(VectorCurve=DataInterface() {propertyValues {
         XCurve {
          defaultValue=0
          keys [
           {interpMode=RCIM_Cubic, tangentMode=RCTM_User, time=0, value=0.25}
           {interpMode=RCIM_Cubic, tangentMode=RCTM_User, time=0.25, value=1.25}
           {arriveTangent=0.5, interpMode=RCIM_Cubic, leaveTangent=0.5, time=0.5, value=0.5}
           <at 0.75 1.5 RCIM_Cubic>
           {tangentMode=RCTM_User, time=1, value=0.5}
          ]
         }
         YCurve {
          defaultValue=0
          keys [
           {interpMode=RCIM_Cubic, tangentMode=RCTM_User, time=0, value=0.25}
           {interpMode=RCIM_Cubic, tangentMode=RCTM_User, time=0.25, value=1.25}
           {arriveTangent=0.5, interpMode=RCIM_Cubic, leaveTangent=0.5, time=0.5, value=0.5}
           <at 0.75 1.5 RCIM_Cubic>
           {tangentMode=RCTM_User, time=1, value=0.5}
          ]
         }
         ZCurve {
          defaultValue=0
          keys [
           {interpMode=RCIM_Cubic, tangentMode=RCTM_User, time=0, value=0.25}
           {interpMode=RCIM_Cubic, tangentMode=RCTM_User, time=0.25, value=1.25}
           {arriveTangent=0.5, interpMode=RCIM_Cubic, leaveTangent=0.5, time=0.5, value=0.5}
           <at 0.75 1.5 RCIM_Cubic>
           {tangentMode=RCTM_User, time=1, value=0.5}
          ]
         }
        }})
     )
     4=<0 0 0>
     Index=&SpawnGroup()
     Num=::NewEnumerator3
    ))
   &UpdateMeshOrientation(
    Rotation Rate=&MergeHelper_FloatSwitch(0=1.75, 1=1.75, 2=3, 3=1.75, 4=0, Index=&SpawnGroup(), Num=::NewEnumerator3)
    Rotation Vector=&RandomRangeVector(Minimum=<-1 -1 -1>, Maximum=<1 1 1>)
   )
   &SolveForcesAndVelocity(Clamp Velocity=false, Limit Acceleration=false)
   &DynamicMaterialParameters(
    Index 0 Param 1=1
    Index 0 Param 2=1
    Index 1 Param 2=1
    Module.Index0Float0Write=true
    Module.Index0Float1Write=true
    Module.Index0Float2Write=true
    Module.Index0Float3Write=true
    Module.Index1Float0Write=true
    Module.Index1Float1Write=true
    Module.Index1Float2Write=true
    Module.Index1Float3Write=true
    Write Parameter Index 1=::NewEnumerator1
   )
  ]
  rendererClasses [&NiagaraSpriteRendererProperties, &NiagaraMeshRendererProperties]
  renderers [
   {rendererClass=&NiagaraSpriteRendererProperties}
   {rendererClass=&NiagaraMeshRendererProperties, rendererIndex=1}
   {rendererClass=&NiagaraMeshRendererProperties, rendererIndex=2}
   {rendererClass=&NiagaraMeshRendererProperties, rendererIndex=3}
   {rendererClass=&NiagaraMeshRendererProperties, rendererIndex=4}
  ]
 }
 NE_VC_SingleBoost_Sprite_Parent001 {
  emitterUpdateScript [
   &EmitterState()
   &SpawnRate(SpawnRate=50, Use Spawn Probability=false)
   &SpawnRate(SpawnGroup=1, SpawnRate=60, Use Spawn Probability=false) {module=SpawnRate001}
  ]
  particleSpawnScript [
   /SetVariables(Particles.VisibilityTag=$Emitter.SpawnGroup) {isSetParameters=true}
   &VC_PaintedVariantsOverride(
    DefaultColor=&SelectLinearColorFromArray(
     Array Sampling Mode=::NewEnumerator1
     Direct Array Index=&RandomRangeInt(Minimum=0, Maximum=1)
     Color Selection Array=DataInterface() {propertyValues {ColorData [<rgba 2.0939 0 5 1>, <rgba 0 3.4947 5 1>]}}
     Random Seed=?
    )
    R1 DefaultColor=<rgba 1 1 1 50>
    Use R1 Painted Overrides=true
   )
   &InitializeParticle(
    Color=$Particles.Color
    Lifetime=&MergeHelper_FloatSwitch(
     0=&RandomRangeFloat(Minimum=0.25, Maximum=0.37)
     1=&RandomRangeFloat(Minimum=1.1, Maximum=1.3)
     2=?
     3=?
     Index=$Emitter.SpawnGroup
     Num=::NewEnumerator0
    )
    Position Offset=&AddVectorToPosition(Position=<-250 0 25>, Vector=$User.DTCarOffset)
    Position Offset Coordinate Space=::Local
    Sprite Rotation Mode=::NewEnumerator2
    Sprite Size Mode=::NewEnumerator3
    Uniform Sprite Size=&MergeHelper_FloatSwitch(
     0=&RandomRangeFloat(Minimum=6, Maximum=18)
     1=&RandomRangeFloat(Minimum=8, Maximum=12)
     2=?
     3=?
     Index=$Emitter.SpawnGroup
     Num=::NewEnumerator0
    )
    UsePositionOffset=true
    Write Material Random=true
   )
   &AddVelocity(Velocity=&MergeHelper_VectorSwitch(0=<-1500 0 0>, 1=<-750 0 0>, 2=?, 3=?, Index=$Emitter.SpawnGroup, Num=::NewEnumerator0))
   &ShapeLocation(
    MaskSpawnGroup=false
    Override Seed=false
    Sphere Radius=&MergeHelper_FloatSwitch(0=75, 1=85, 2=?, 3=?, Index=$Emitter.SpawnGroup, Num=::NewEnumerator0)
    UseOwnerScale=true
   )
  ]
  particleUpdateScript [
   &ParticleState()
   &SolveForcesAndVelocity(Clamp Velocity=false, Limit Acceleration=false)
   &ScaleColor(
    Scale Alpha=&FloatFromCurve(FloatCurve=DataInterface() {propertyValues {Curve {keys [<at 0 1>,<at 0.5 1>,<at 1 0>]}}})
    ScaleA=true
    ScaleRGB=true
   )
  ]
  rendererClasses [&NiagaraSpriteRendererProperties]
  renderers [
   {rendererClass=&NiagaraSpriteRendererProperties}
   {rendererClass=&NiagaraSpriteRendererProperties, rendererIndex=1}
  ]
 }
}
summary {
 systemName=NS_Supernova
 emitters [
  {emitterName=NE_VC_SingleBoost_Ribbon_Parent, rendererClasses [&NiagaraRibbonRendererProperties]}
  {
   emitterName=NE_VC_SingleBoost_Sprite_Parent
   rendererClasses [&NiagaraSpriteRendererProperties, &NiagaraMeshRendererProperties]
  }
  {emitterName=NE_VC_SingleBoost_Sprite_Parent001, rendererClasses [&NiagaraSpriteRendererProperties]}
  {emitterName=NE_VC_SingleBoost_Ribbon_Parent001, rendererClasses [&NiagaraRibbonRendererProperties]}
 ]
 userVariables [
  User.bIsFrontendPreview {type=NiagaraBool}
  User.bIsPainted {type=NiagaraBool}
  User.bPreviewVehicleSpeed {type=NiagaraBool}
  User.ColorVariantID {type=NiagaraInt32}
  User.DTCarOffset {type=Vector3f, description="", defaultValue {struct=Vector3f, value=<0 0 0>}}
  User.GameModeID {type=NiagaraInt32}
  User.PaintedColor {type=LinearColor, description="", defaultValue {struct=LinearColor, value=<rgba 1 1 1 1>}}
  User.PaintedScalar {type=NiagaraFloat, defaultValue=1, description=""}
  User.PortCount {type=NiagaraInt32, defaultValue=1}
  User.SingleBoostPos {type=NiagaraPosition, description="", defaultValue {}}
 ]
}
