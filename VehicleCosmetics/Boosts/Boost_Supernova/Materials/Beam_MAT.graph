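# Translucent unlit beam material for the Supernova boost effect, built for Niagara ribbons and sprites.
# A parabola across the V texture coordinate is thresholded steeply into two centered bands, a wide body and a narrow core, without any textures.
# The core mask plus a constant floor scales particle color, giving a bright center over dimmer edges, then feeds MF_FX_EmissiveAdjustments.
# The wide band forms a separate saturated mask.
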Material /VehicleCosmetics/Boosts/Boost_Supernova/Materials/Beam_MAT
graph {
 properties {
  bUsedWithNiagaraRibbons=true
  bUsedWithNiagaraSprites=true
  blendMode=BLEND_Translucent
  pixelDepthOffsetMode=PDOM_Legacy
  shadingModel=MSM_Unlit
  usedShadingModels=Unlit
 }
 nodes [
  n1 = Constant() {r=64}
  n2 = ComponentMask(None=TextureCoordinate()) {g=true}
  n3 = Multiply(A=@n2, B=OneMinus(None=@n2))
  n4 = Constant() {r=0.1}
  MaterialExpressionConstantClamp_0 = Saturate(None=Multiply(A=@n1, B=Subtract(A=@n3, B=@n4)))
  MaterialExpressionVertexColor_0 = ParticleColor()
  n5 = Constant() {r=0.2}
  MaterialExpressionConstantClamp_2 = Saturate(None=Multiply(A=Subtract(A=@n3, B=@n5), B=@n1))
  MaterialExpressionConstantClamp_1 = Saturate(None=Add(A=Constant() {r=0.25}, B=@MaterialExpressionConstantClamp_2))
  MaterialFunctionCall(Emissive=Multiply(A=@MaterialExpressionVertexColor_0.RGB, B=@MaterialExpressionConstantClamp_1)) {materialFunction=&MF_FX_EmissiveAdjustments}
 ]
}
