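# Masked, unlit, two-sided Niagara ribbon material for the electric arcs of the Supernova vehicle boost.
# Dynamic parameters offset, desaturate and brighten a rainbow texture for emissive, with an optional black outline masked from UV V.
# Zero-tiled UVs pan one sample point through cloud noise, thresholded into a hard flicker at Tile-driven speed.
# Sine and cosine of actor-relative position and time build a transformed wobble vector, with particle RGB setting wavelength, speed and amplitude.
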Material /VehicleCosmetics/Boosts/Boost_Supernova/Materials/ElectricArc_MAT
graph {
 properties {
  bUsedWithNiagaraRibbons=true
  blendMode=BLEND_Masked
  pixelDepthOffsetMode=PDOM_Legacy
  shadingModel=MSM_Unlit
  twoSided=true
  usedShadingModels=Unlit
 }
 nodes [
  n1 = DynamicParameter() {paramNames [color, Color_Brightness, Desaturate, Tile]}
  n2 = Time()
  n3 = TextureSample(UVs=Panner(Coordinate=TextureCoordinate() {coordinateIndex=1, uTiling=0, vTiling=0}, Time=Multiply(A=@n1.Tile, B=@n2)) {speedX=2, speedY=3}) {texture=/VehicleCosmetics/FX/Textures/Noise/T_VCVFX_Clouds_A}
  n4 = Constant() {r=0.4}
  MaterialExpressionConstantClamp_1 = Saturate(None=Multiply(A=Subtract(A=@n3.RGB, B=@n4), B=Constant() {r=55}))
  n5 = ScalarParameter() {defaultValue=1, parameterName=UseBlackOutline?}
  n6 = ComponentMask(None=TextureCoordinate()) {g=true}
  n7 = Multiply(A=@n6, B=OneMinus(None=@n6))
  n8 = Constant() {r=0.2}
  MaterialExpressionConstantClamp_2 = Saturate(None=Ceil(None=Subtract(A=@n7, B=@n8)))
  n9 = Constant() {r=1}
  n10 = TextureCoordinate() {vTiling=0}
  n11 = Desaturation(
   None=TextureSample(UVs=Add(A=@n10, B=@n1.color)) {texture=/DealershipSharedLibrary/FX/Textures/Diffuse/T_VCVFX_Rainbow_D}.RGB
   Fraction=@n1.Desaturate
  )
  MaterialFunctionCall(Emissive=Multiply(
    A=If(A=@n5, "A > B"=@MaterialExpressionConstantClamp_2, "A == B"=@MaterialExpressionConstantClamp_2, "A < B"=@n9) {constB=1}
    B=Multiply(A=@n11, B=@n1.Color_Brightness)
   )) {materialFunction=&MF_FX_EmissiveAdjustments}
  MaterialExpressionVertexColor_1 = ParticleColor()
  n12 = Multiply(A=@MaterialExpressionVertexColor_1.G, B=OneMinus(None=Time()))
  n13 = ActorPositionWS()
  n14 = WorldPosition()
  n15 = Divide(A=Abs(None=Subtract(A=@n13, B=@n14.XYZ)), B=@MaterialExpressionVertexColor_1.R)
  n16 = Add(A=Add(A=ComponentMask(None=@n15) {r=true}, B=ComponentMask(None=@n15) {g=true}), B=ComponentMask(None=@n15) {b=true})
  n17 = Add(A=@n12, B=@n16)
  n18 = Multiply(A=Cosine(None=@n17) {period=1.25}, B=Sine(None=@n17))
  n19 = AppendVector(A=@n18, B=Sine(None=Add(A=@n16, B=@n12)))
  Transform(None=Multiply(A=AppendVector(A=@n19, B=@n18), B=@MaterialExpressionVertexColor_1.B))
 ]
}
