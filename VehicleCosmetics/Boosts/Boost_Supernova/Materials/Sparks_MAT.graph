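# Unlit, two-sided translucent Niagara sprite material for sparks in the Supernova boost effect.
# Particle color RGB is scaled by a Darkness dynamic parameter, letting the particle system control brightness, then passed into the shared MF_FX_EmissiveAdjustments function.
# It also samples a shared ash mask texture, though the links to the final material outputs are not shown.
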Material /VehicleCosmetics/Boosts/Boost_Supernova/Materials/Sparks_MAT
graph {
 properties {
  bUsedWithNiagaraSprites=true
  blendMode=BLEND_Translucent
  pixelDepthOffsetMode=PDOM_Legacy
  shadingModel=MSM_Unlit
  twoSided=true
  usedShadingModels=Unlit
 }
 nodes [
  MaterialExpressionVertexColor_3 = ParticleColor()
  MaterialFunctionCall(Emissive=Multiply(A=DynamicParameter() {paramNames [Darkness,Param2,Param3,Param4]}.Darkness, B=@MaterialExpressionVertexColor_3.RGB)) {materialFunction=&MF_FX_EmissiveAdjustments}
  TextureSample() {texture=/DealershipSharedLibrary/FX/Textures/Masks/T_VCVFX_Ash_A}
 ]
}
