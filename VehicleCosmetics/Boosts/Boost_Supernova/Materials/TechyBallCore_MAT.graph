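# TechyBallCore_MAT is a translucent, two-sided unlit Niagara mesh-particle material for the Supernova boost's ball core.
# Emissive is particle color times CoreBrightness through MF_FX_EmissiveAdjustments; the edge-stripe lerp has identical inputs and the reflection term is zeroed.
# Pivot Painter data rotates pieces about per-object pivots, with particle alpha scaling pivot offsets and Alt Spin blending two rotation variants.
# A stepped mask compares particle or vertex alpha against RingCount.
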Material /VehicleCosmetics/Boosts/Boost_Supernova/Materials/TechyBallCore_MAT
graph {
 properties {
  bUsedWithNiagaraMeshParticles=true
  blendMode=BLEND_Translucent
  pixelDepthOffsetMode=PDOM_Legacy
  shadingModel=MSM_Unlit
  twoSided=true
  usedShadingModels=Unlit
 }
 nodes [
  n1 = AppendVector(A=TextureCoordinate() {coordinateIndex=2}, B=ComponentMask(None=TextureCoordinate() {coordinateIndex=3}) {r=true})
  n2 = Constant() {r=1}
  ComponentMask(None=TransformPosition(None=AppendVector(A=@n1, B=@n2)) {transformType=TRANSFORMPOSSOURCE_World}) {b=true, g=true, r=true}
  Constant() {r=1}
  MaterialExpressionMeshEmitterVertexColor_0 = ParticleColor()
  n3 = VertexColor()
  n4 = ScalarParameter() {parameterName=Vertex Color}
  LinearInterpolate(A=@MaterialExpressionMeshEmitterVertexColor_0.A, B=@n3.A, Alpha=@n4)
  LinearInterpolate(A=@MaterialExpressionMeshEmitterVertexColor_0.RGB, B=@n3, Alpha=@n4)
  MaterialExpressionMeshEmitterDynamicParameter_0 = DynamicParameter() {paramNames [RingCount, CoreBrightness, spin, edgeBrightness]}
  n5 = MaterialFunctionCall() {materialFunction=/Engine/Functions/Engine_MaterialFunctions02/PivotPainter/PivotPainter_PerObjectData}
  n6 = AppendVector(A=@n5.X-Axis Vector, B=@MaterialExpressionMeshEmitterDynamicParameter_0.spin)
  n7 = Multiply(A=Subtract(A=WorldPosition().XYZ, B=@n5.Pivot Position), B=@MaterialExpressionMeshEmitterVertexColor_0.A)
  n8 = Add(A=WorldPosition().XYZ, B=@n7)
  n9 = MaterialFunctionCall() {materialFunction=&PivotPainter_PerObjectData}
  n10 = DynamicParameter() {paramNames [RingCount, CoreBrightness, spin, edgeBrightness]}
  n11 = AppendVector(A=@n9.X-Axis Vector, B=@n10.spin)
  n12 = Multiply(A=Subtract(A=WorldPosition().XYZ, B=@n9.Pivot Position), B=@MaterialExpressionMeshEmitterVertexColor_0.A)
  n13 = Add(A=WorldPosition().XYZ, B=@n12)
  LinearInterpolate(
   A=Add(A=RotateAboutAxis(NormalizedRotationAxis=@n6, RotationAngle=@n5.Pivot Position, PivotPoint=@n5.Pivot Position, Position=@n8), B=@n7)
   B=Add(A=RotateAboutAxis(NormalizedRotationAxis=@n11, RotationAngle=@n10.spin, PivotPoint=@n9.Pivot Position, Position=@n13), B=@n12)
   Alpha=ScalarParameter() {parameterName=Alt Spin}
  )
  n14 = ReflectionVectorWS()
  n15 = Constant()
  n16 = Frac(None=TextureCoordinate() {uTiling=4, vTiling=0})
  n17 = Constant() {r=-0.5}
  n18 = Abs(None=ComponentMask(None=Add(A=@n16, B=@n17)) {r=true})
  n19 = ScalarParameter() {defaultValue=0.47, parameterName=Edge Width}
  n20 = LinearInterpolate(
   A=@MaterialExpressionMeshEmitterDynamicParameter_0.CoreBrightness
   B=@MaterialExpressionMeshEmitterDynamicParameter_0.CoreBrightness
   Alpha=Ceil(None=Subtract(A=@n18, B=@n19))
  )
  MaterialFunctionCall(Emissive=Add(A=Multiply(A=@n14, B=@n15), B=Multiply(A=@MaterialExpressionMeshEmitterVertexColor_0.RGB, B=@n20))) {materialFunction=&MF_FX_EmissiveAdjustments}
  MaterialExpressionVertexColor_0 = ParticleColor()
  n21 = LinearInterpolate(A=@MaterialExpressionVertexColor_0.A, B=VertexColor().A, Alpha=ScalarParameter() {parameterName=Vertex Color})
  MaterialExpressionConstantClamp_0 = Saturate(None=Ceil(None=Subtract(A=@n21, B=@MaterialExpressionMeshEmitterDynamicParameter_0.RingCount)))
  Multiply(A=Constant() {r=0.1}, B=@MaterialExpressionConstantClamp_0)
  n22 = ConstantBiasScale(None=VertexColor()) {bias=-0.5, scale=2}
  Normalize(VectorInput=Transform(None=Normalize(VectorInput=@n22)) {transformSourceType=TRANSFORMSOURCE_Local})
 ]
}
