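# Unlit translucent Niagara mesh-particle material for the Supernova boost's techy ring, drawing lines and dots in one mesh.
# The rounded-up UV3 green channel selects between line and dot.
# Dot quads become camera-facing billboards, rebuilt around pivots stored in UV2 and UV3 red and sized by dynamic radius.
# Opacity uses sharp procedural masks, a thickness-driven strip or a circle, times particle alpha; particle color feeds emissive through MF_FX_EmissiveAdjustments.
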Material /VehicleCosmetics/Boosts/Boost_Supernova/Materials/TechyRing_LineAndParticle_MAT
graph {
 properties {
  bUsedWithNiagaraMeshParticles=true
  blendMode=BLEND_Translucent
  pixelDepthOffsetMode=PDOM_Legacy
  shadingModel=MSM_Unlit
  twoSided=true
  usedShadingModels=Unlit
 }
 nodes [
  n1 = AppendVector(A=TextureCoordinate() {coordinateIndex=2}, B=ComponentMask(None=TextureCoordinate() {coordinateIndex=3}) {r=true})
  n2 = Constant() {r=1}
  n3 = ComponentMask(None=TransformPosition(None=AppendVector(A=@n1, B=@n2)) {transformType=TRANSFORMPOSSOURCE_World}) {b=true, g=true, r=true}
  MaterialExpressionMeshEmitterDynamicParameter_1 = DynamicParameter() {paramNames [radius,thickness,Param3,Param4]}
  n4 = Multiply(A=@MaterialExpressionMeshEmitterDynamicParameter_1.radius, B=Constant() {r=-1})
  MaterialExpressionCameraWorldPosition_1 = CameraPositionWS()
  n5 = Normalize(VectorInput=Subtract(A=@MaterialExpressionCameraWorldPosition_1, B=@n3))
  n6 = CrossProduct(A=@n5, B=Normalize(VectorInput=Constant3Vector() {constant=<rgba 0 0 1 0>}))
  n7 = Normalize(VectorInput=CrossProduct(A=@n6, B=@n5))
  n8 = ConstantBiasScale(None=TextureCoordinate()) {bias=-0.5, scale=2}
  n9 = Normalize(VectorInput=@n6)
  n10 = Add(
   A=Multiply(A=Multiply(A=@n7, B=@n4), B=ComponentMask(None=@n8) {r=true})
   B=Multiply(A=Multiply(A=@n9, B=@MaterialExpressionMeshEmitterDynamicParameter_1.radius), B=ComponentMask(None=@n8) {g=true})
  )
  n11 = WorldPosition()
  n12 = Ceil(None=ComponentMask(None=TextureCoordinate() {coordinateIndex=3}) {g=true})
  LinearInterpolate(A=Constant3Vector() {constant=<rgba 0 0.5 0 0>}, B=Add(A=Subtract(A=@n10, B=@n11.XYZ), B=@n3), Alpha=@n12)
  MaterialExpressionMeshEmitterVertexColor_1 = ParticleColor()
  MaterialFunctionCall(Emissive=@MaterialExpressionMeshEmitterVertexColor_1.RGB) {materialFunction=&MF_FX_EmissiveAdjustments}
  n13 = Add(A=TextureCoordinate(), B=Constant() {r=-0.5})
  n14 = OneMinus(None=Abs(None=ComponentMask(None=@n13) {g=true}))
  n15 = Constant() {r=512}
  MaterialExpressionConstantClamp_5 = Saturate(None=Multiply(A=Subtract(A=@n14, B=@MaterialExpressionMeshEmitterDynamicParameter_1.thickness), B=@n15))
  n16 = OneMinus(None=DotProduct(A=@n13, B=@n13))
  n17 = Constant() {r=0.8}
  MaterialExpressionConstantClamp_6 = Saturate(None=Multiply(A=Subtract(A=@n16, B=@n17), B=@n15))
  Multiply(
   A=@MaterialExpressionMeshEmitterVertexColor_1.A
   B=LinearInterpolate(A=@MaterialExpressionConstantClamp_5, B=@MaterialExpressionConstantClamp_6, Alpha=@n12)
  )
  VertexColor()
 ]
}
