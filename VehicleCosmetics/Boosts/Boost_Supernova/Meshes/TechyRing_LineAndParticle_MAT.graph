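# Material for the TechyRing line and particle mesh in the Supernova vehicle boost cosmetic, flagged for use on Niagara mesh particles.
# It uses Default Lit shading, and its graph holds a single unnamed vector parameter defaulting to white with zero alpha.
# No connection from that parameter to a material output is shown.
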Material /VehicleCosmetics/Boosts/Boost_Supernova/Meshes/TechyRing_LineAndParticle_MAT
graph {
 properties {bUsedWithNiagaraMeshParticles=true, pixelDepthOffsetMode=PDOM_Legacy, usedShadingModels=Default Lit}
 nodes [VectorParameter() {defaultValue=<rgba 1 1 1 0>}]
}
